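; Gameplay ability that applies the IcicleLarva punch knockback, with separate tuning when the avatar is a Fortnite vehicle.
; On activation it checks whether the avatar is a FortAthenaVehicle and, if so, scales the inherited XY magnitude, Z magnitude and Z add by a vehicle multiplier.
; The scaled values overwrite the parent's scalable float variables in place before the parent activation runs, which happens for all avatars.
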(event Ability|EventActivateAbility
  ;[1 1 1] Vehicles could have different knockback than other actors
  (if (Utilities|IsA (Ability|GetAvatarActorfromActorInfo) "/Script/FortniteGame.FortAthenaVehicle")
    (CallFunction|UpdateKnockbackValuesforVehicle))
  (|CallParentFunction))

(fn UpdateKnockbackValuesForVehicle
  (LocalVariables|SetKnockbackMultiplier (ScalableFloat|GetValueatLevel (Variables|Default|GetVehicleKnockbackMultiplier)))
  (Variables|KnockbackParent|SetRow_XYKnockMag (Utilities|Struct|MakeScalableFloat (* (ScalableFloat|GetValueatLevel (Variables|KnockbackParent|GetRow_XYKnockMag)) (LocalVariables|GetKnockbackMultiplier))))
  (Variables|KnockbackParent|SetRow_ZKnockMag (Utilities|Struct|MakeScalableFloat (* (ScalableFloat|GetValueatLevel (Variables|KnockbackParent|GetRow_ZKnockMag)) (LocalVariables|GetKnockbackMultiplier))))
  (Variables|KnockbackParent|SetRow_ZAdd (Utilities|Struct|MakeScalableFloat (* (ScalableFloat|GetValueatLevel (Variables|KnockbackParent|GetRow_ZAdd)) (LocalVariables|GetKnockbackMultiplier)))))
